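# Translucent unlit master-style FX material for the Tower Sentinel Female glider's flame-like energy wing, usable on meshes and Niagara particles.
# Distorted panning noise, shape masks and depth fade build opacity, with parameters for appearance and optional stair masking.
# Emissive blends textured base colors, curve-atlas gradient and hot center, tinted by time of day.
# Noise WPO scales with camera distance, plus optional Niagara-driven bending; a static switch substitutes pickaxe-trail shading.
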Material /BRCosmetics/Gadgets/Assets/VinderTech_GliderChute/Glider_Tower_Sentinel_Female/FX/M_Glider_Tower_Sentinel_Female_wing
graph {
 properties {
  bUsedWithMeshDeformer=true
  bUsedWithNiagaraMeshParticles=true
  bUsedWithNiagaraRibbons=true
  bUsedWithNiagaraSprites=true
  bUsedWithSkeletalMesh=true
  blendMode=BLEND_Translucent
  pixelDepthOffsetMode=PDOM_Legacy
  shadingModel=MSM_Unlit
  usedShadingModels=Unlit
 }
 nodes [
  n1 = ComponentMask(None=TextureCoordinate()) {g=true}
  n2 = Add(A=LinearInterpolate(Alpha=OneMinus(None=@n1)) {constA=-1}, B=Constant() {r=0.8})
  n3 = OneMinus(None=SphereMask(
    A=TextureCoordinate() {uTiling=1.5, vTiling=0.8}
    B=Constant2Vector() {g=0.8, r=0.75}
    Radius=Constant() {r=0.5}
    Hardness=Constant() {r=0.5}
   ))
  Add(A=@n2, B=@n3)
  n4 = Multiply(
   A=TextureCoordinate()
   B=AppendVector(A=Constant() {r=1}, B=DynamicParameter() {parameterIndex=2, paramNames [V_Scale,Param2,Param3,Param4]}.V_Scale)
  )
  n5 = DynamicParameter() {paramNames [Offset, Texture_mask_offset, WaveIntensity, Panner_offset]}
  n6 = Panner(
   Coordinate=LinearInterpolate(A=TextureCoordinate() {uTiling=0.7, vTiling=0.8}, B=Add(A=@n4, B=@n5.Panner_offset), Alpha=ScalarParameter() {parameterName=Fix_the_edges})
   Time=LinearInterpolate(
    A=DynamicParameter() {paramNames [Offset, Texture_mask_offset, WaveIntensity, Panner_offset]}.Texture_mask_offset
    B=Time()
    Alpha=ScalarParameter() {group=Base_color, parameterName=Loop_panner}
   )
  ) {speedY=0.2}
  n7 = ComponentMask(None=VectorParameter() {group=Distortion, parameterName=Distortion UV Scale, sortPriority=3, defaultValue=<rgba 1 1 0 0>}.RGB) {g=true, r=true}
  n8 = TextureCoordinate() {vTiling=1.5}
  n9 = Multiply(A=Add(A=Multiply(A=@n7, B=@n8), B=NamedRerouteUsage()))
  n10 = VertexColor()
  n11 = AppendVector(A=Constant(), B=@n10.R)
  n12 = Time()
  n13 = ScalarParameter() {defaultValue=1, group=Panner Speed, parameterName=Speed}
  n14 = TextureSample(UVs=Panner(Coordinate=Add(A=@n9, B=@n11), Time=Multiply(A=@n12, B=@n13)) {speedY=1}) {samplerType=SAMPLERTYPE_Normal, texture=&T_Weap_Distort_N_02}
  n15 = TextureCoordinate()
  n16 = VectorParameter() {group=Fire Shape, parameterName=Fire_MacroShape_UVScale, sortPriority=5, defaultValue=<rgba 1 1 0 0>}
  n17 = ComponentMask(None=@n16.RGB) {g=true, r=true}
  n18 = Multiply(A=@n15, B=@n17)
  n19 = Divide(A=LinearInterpolate(A=ComponentMask(None=@n14.RGB) {g=true, r=true}, Alpha=ComponentMask(None=@n18) {g=true}) {constB=0}, B=Constant() {r=3}) {constB=20}
  n20 = TextureCoordinate() {uTiling=0.9, vTiling=0.8}
  n21 = NamedRerouteUsage()
  n22 = Multiply(
   A=TextureCoordinate()
   B=AppendVector(A=Constant() {r=1}, B=DynamicParameter() {parameterIndex=2, paramNames [V_Scale,Param2,Param3,Param4]}.V_Scale)
  )
  n23 = DynamicParameter() {paramNames [Offset, Texture_mask_offset, WaveIntensity, Panner_offset]}
  n24 = Multiply(
   A=LinearInterpolate(A=Add(A=@n20, B=@n21), B=Add(A=@n22, B=@n23.Panner_offset), Alpha=ScalarParameter() {parameterName=Fix_the_edges})
   B=ComponentMask(None=VectorParameter() {group=Distortion, parameterName=Fire Micro UV Scale, sortPriority=3, defaultValue=<rgba 1 1 0 0>}.RGB) {g=true, r=true}
  )
  n25 = Constant()
  n26 = VertexColor()
  n27 = Panner(
   Coordinate=Add(A=Multiply(A=@n24), B=AppendVector(A=@n25, B=@n26.R))
   Time=Multiply(A=Time(), B=ScalarParameter() {defaultValue=1, group=Panner Speed, parameterName=Speed})
  ) {speedY=1.2}
  n28 = ChannelMaskParameter(None=TextureSample(UVs=Add(A=@n19, B=@n27)) {samplerType=SAMPLERTYPE_LinearColor, texture=&T_NoisePack_LinearColor}.RGBA) {group=Fire Shape, parameterName=Fire Micro Mask, sortPriority=7}
  Add(A=@n6, B=Multiply(A=@n28.R, B=ScalarParameter() {group=Base_color, parameterName=Base_texture_disturb}))
  AppendVector(A=Constant() {r=1}, B=@n10.R)
  n29 = VertexColor()
  AppendVector(A=Constant() {r=1}, B=@n29.R)
  Constant() {r=0.3}
  Frac(None=Time())
  n30 = VectorParameter() {group=Color_time_of_day, parameterName=Morning_Color, defaultValue=<rgba 1 1 1 1>}
  n31 = VectorParameter() {group=Color_time_of_day, parameterName=Day_Color, defaultValue=<rgba 1 1 1 1>}
  n32 = VectorParameter() {group=Color_time_of_day, parameterName=Evening_Color, defaultValue=<rgba 1 1 1 1>}
  n33 = VectorParameter() {group=Color_time_of_day, parameterName=Night_Color, defaultValue=<rgba 1 1 1 1>}
  n34 = ParticleColor()
  n35 = ScalarParameter() {defaultValue=-0.3, group=Opacity, parameterName=Mask_Dark}
  n36 = TextureCoordinate()
  n37 = Sine(None=ComponentMask(None=@n36) {r=true}) {period=2}
  n38 = LinearInterpolate(Alpha=ComponentMask(None=@n36) {g=true}) {constA=50, constB=10}
  n39 = ComponentMask(None=VectorParameter() {group=Opacity, parameterName=Mask_UV_Scale, defaultValue=<rgba 1 1 0 1>}.RGB) {g=true, r=true}
  n40 = Multiply(A=TextureCoordinate(), B=@n39)
  n41 = AppendVector(A=Constant() {r=0.6}, B=VertexColor().R)
  n42 = NamedRerouteUsage()
  n43 = TextureSample(UVs=Panner(Coordinate=Add(A=@n40, B=@n41) {constB=0.6}, Time=@n42) {speedY=0.5}) {texture=/Game/CharClasses/Outlander/Gadgets/AntiMaterialCharge/Materials/T_AntiMaterialCharge_Dissolve_01}
  n44 = Multiply(
   A=TextureCoordinate()
   B=AppendVector(A=Constant() {r=1}, B=DynamicParameter() {parameterIndex=2, paramNames [V_Scale,Param2,Param3,Param4]}.V_Scale)
  )
  n45 = DynamicParameter() {paramNames [Offset, Texture_mask_offset, WaveIntensity, Panner_offset]}
  n46 = Multiply(
   A=@n39
   B=LinearInterpolate(A=TextureCoordinate() {uTiling=0.8}, B=Add(A=@n44, B=@n45.Panner_offset), Alpha=ScalarParameter() {parameterName=Fix_the_edges})
  )
  n47 = AppendVector(A=Constant(), B=VertexColor().R)
  n48 = TextureSample(UVs=Panner(Coordinate=Add(A=@n46, B=@n47), Time=@n42) {speedY=0.5}) {texture=&T_AntiMaterialCharge_Dissolve_01}
  n49 = Add(A=Saturate(None=Power(Base=@n37, Exp=@n38) {constExponent=10}), B=OneMinus(None=LinearInterpolate(A=@n43.R, B=@n48.R)))
  n50 = Constant() {r=0.3}
  n51 = TextureCoordinate() {uTiling=0.5, vTiling=0.9}
  n52 = NamedRerouteUsage()
  n53 = Multiply(
   A=TextureCoordinate()
   B=AppendVector(A=Constant() {r=1}, B=DynamicParameter() {parameterIndex=2, paramNames [V_Scale,Param2,Param3,Param4]}.V_Scale)
  )
  n54 = DynamicParameter() {paramNames [Offset, Texture_mask_offset, WaveIntensity, Panner_offset]}
  n55 = Multiply(A=LinearInterpolate(A=Add(A=@n51, B=@n52), B=Add(A=@n53, B=@n54.Panner_offset), Alpha=ScalarParameter() {parameterName=Fix_the_edges}))
  n56 = AppendVector(A=Constant(), B=@n29.R)
  n57 = TextureSample(UVs=Panner(Coordinate=Add(A=@n55, B=@n56), Time=NamedRerouteUsage()) {speedY=0.5}) {texture=&T_ShieldCaustics}
  n58 = LinearInterpolate(Alpha=LinearInterpolate(Alpha=Power(Base=@n57.R) {constExponent=0.6}) {constB=3}) {constA=-0.3}
  n59 = TextureCoordinate()
  n60 = TextureCoordinate()
  n61 = ComponentMask(None=@n60) {r=true}
  n62 = ComponentMask(None=@n60) {g=true}
  n63 = Multiply(
   A=Power(Base=ComponentMask(None=@n59) {g=true}) {constExponent=4}
   B=Power(Base=Sine(None=@n61) {period=2}, Exp=LinearInterpolate(Alpha=@n62) {constA=50, constB=0.5}) {constExponent=10}
  )
  n64 = OneMinus(None=ComponentMask(None=TextureCoordinate()) {g=true})
  n65 = Multiply(A=Add(A=Saturate(None=@n58), B=Saturate(None=@n63)) {constB=0}, B=Saturate(None=Multiply(A=@n64) {constB=3}))
  n66 = Multiply(A=@n65) {constB=0.3}
  n67 = MaterialFunctionCall(Float2=Multiply(A=@n15, B=@n17)) {materialFunction=&BreakOutFloat2Components}
  n68 = Sine(None=Multiply(A=@n67.R) {constB=0.5})
  n69 = LinearInterpolate(
   A=ScalarParameter() {defaultValue=2, group=Fire Shape, parameterName=Edge Fade Value Top, sortPriority=3}
   B=ScalarParameter() {defaultValue=1, group=Fire Shape, parameterName=Edge Fade Value Bottom, sortPriority=4}
   Alpha=@n67.G
  ) {constA=6, constB=2}
  n70 = VectorParameter() {group=BottomFade, parameterName=Bottom Fade_SphereMask_A_Scale, sortPriority=2, defaultValue=<rgba 1.2 1 0 0>}
  n71 = Multiply(A=TextureCoordinate(), B=ComponentMask(None=@n70.RGB) {g=true, r=true})
  n72 = VectorParameter() {group=BottomFade, parameterName=Bottom Fade_SphereMask_B, sortPriority=3, defaultValue=<rgba 0.6 0.05 0 0>}
  n73 = ComponentMask(None=@n72.RGB) {g=true, r=true}
  n74 = ScalarParameter() {defaultValue=0.95, group=BottomFade, parameterName=BottomFade_Radius, sortPriority=1}
  n75 = OneMinus(None=ComponentMask(None=TextureCoordinate()) {g=true})
  n76 = ScalarParameter() {defaultValue=1, parameterName=StraightBottomFade_Mult}
  n77 = LinearInterpolate(
   A=Multiply(
    A=SphereMask(A=@n71, B=@n73, Radius=@n74) {attenuationRadius=0.95, hardnessPercent=0}
    B=ScalarParameter() {defaultValue=15, parameterName=SphereMask_Multiply}
   ) {constB=15}
   B=Power(Base=Multiply(A=@n75, B=@n76), Exp=ScalarParameter() {defaultValue=1, parameterName=StraightBottomFade_Power}) {constExponent=1}
   Alpha=ScalarParameter() {parameterName=UseStraightBottomFade}
  )
  n78 = Multiply(
   A=Multiply(A=Power(Base=@n68, Exp=@n69) {constExponent=4}, B=@n67.G)
   B=LinearInterpolate(B=Saturate(None=@n77), Alpha=ScalarParameter() {defaultValue=1, group=BottomFade, parameterName=Bottom Fade}) {constA=1}
  )
  n79 = StaticSwitchParameter(True=LinearInterpolate(A=@n65, B=@n28.R) {constAlpha=0.6}, False=@n28.R) {group=Opacity, parameterName=Use_detail_smoke}
  n80 = Multiply(A=Power(Base=Multiply(A=@n79, B=@n78), Exp=LinearInterpolate(Alpha=@n67.G) {constA=2}))
  n81 = Multiply(A=@n80, B=ScalarParameter() {defaultValue=6, group=Opacity, parameterName=OpacityMaskMultiplier, sortPriority=1}) {constB=3}
  n82 = Multiply(
   A=StaticSwitchParameter(True=Add(A=@n50, B=@n66), False=LinearInterpolate(A=@n35, Alpha=@n49) {constA=-0.3}) {group=Opacity, parameterName=Use_detail_smoke}
   B=LinearInterpolate(A=Floor(None=@n81), B=@n81, Alpha=ScalarParameter() {defaultValue=1, parameterName=Floor})
  ) {constA=1}
  n83 = ComponentMask(None=@n82) {r=true}
  n84 = LinearInterpolate(Alpha=Multiply(A=@n83) {constB=3})
  n85 = StaticSwitchParameter(True=Saturate(None=LinearInterpolate(A=@n83, B=@n84)), False=@n82) {group=Opacity, parameterName=Use_detail_smoke}
  n86 = ComponentMask(None=TextureCoordinate()) {g=true}
  n87 = Add(A=@n85, B=Saturate(None=Power(Base=@n86) {constExponent=8}))
  n88 = ScalarParameter() {group=Opacity, parameterName=Add_Center_Bottom_mask}
  n89 = Multiply(
   A=Multiply(A=LinearInterpolate(A=@n85, B=@n87, Alpha=@n88), B=ScalarParameter() {defaultValue=1, group=Opacity, parameterName=Global_Opacity})
   B=MaterialFunctionCall("FadeSensitivity (optional)"=ScalarParameter() {group=Opacity, parameterName=FadeBasedOnViewAngle}) {materialFunction=&FadeBasedOnViewAngle}.Result
  )
  n90 = DepthFade(Opacity=@n89, FadeDistance=ScalarParameter() {defaultValue=30, group=Opacity, parameterName=Depth_fade})
  n91 = ShadingPathSwitch(Default=@n90, Deferred=@n90, Forward=@n90, Mobile=@n89)
  n92 = LinearInterpolate(Alpha=ComponentMask(None=TextureCoordinate()) {g=true}) {constA=-1}
  n93 = LinearInterpolate(Alpha=ScalarParameter() {defaultValue=1, parameterName=Glider_appear}) {constA=-1.5, constB=1.5}
  n94 = Multiply(
   A=LinearInterpolate(A=@n91, B=Multiply(A=@n34.A, B=@n91), Alpha=ScalarParameter() {parameterName=Use_particle_color})
   B=Saturate(None=Add(A=@n92, B=@n93))
  ) {constB=0}
  n95 = Multiply(A=@n2, B=@n3)
  n96 = LinearInterpolate(A=@n94, B=Multiply(A=@n94, B=Saturate(None=@n95)), Alpha=ScalarParameter() {parameterName=Use_HeadMask})
  n97 = OneMinus(None=ComponentMask(None=TextureCoordinate()) {g=true})
  n98 = Multiply(A=Power(Base=Add(A=@n97) {constB=0}, Exp=Constant() {r=3}) {constExponent=3}, B=Constant() {r=20}) {constB=20}
  n99 = ActorPositionWS()
  n100 = Constant() {r=-20}
  n101 = Constant() {r=20}
  n102 = LinearInterpolate(Alpha=ParticleRandom()) {constA=5.7, constB=5.3}
  n103 = Time()
  n104 = TextureSample(UVs=Panner(Coordinate=TextureCoordinate(), Time=Add(A=@n102, B=@n103)) {speedY=0.5}) {samplerType=SAMPLERTYPE_Grayscale, texture=&LowResBlurredNoise}
  n105 = Transform(None=Add(
    A=VectorParameter() {group=CubeQueen_Stairs_Mask, parameterName=Stairs_Mask_Offset, defaultValue=<rgba 0 -170 0 1>}.RGB
    B=LinearInterpolate(A=@n100, B=@n101, Alpha=@n104.R) {constA=-50, constB=0}
   ) {constB=0}) {transformSourceType=TRANSFORMSOURCE_Local}
  n106 = Constant() {r=500}
  n107 = Subtract(A=@n106, B=Constant() {r=50}) {constB=50}
  n108 = CurveAtlasRowParameter(CurveTime=SphereMask(
    A=WorldPosition().XYZ
    B=Add(A=@n99, B=@n105)
    Radius=LinearInterpolate(A=@n107, B=@n106, Alpha=@n104.R) {constA=250}
    Hardness=ScalarParameter() {parameterName=Flame_Hardness}
   )) {
   atlas=/Game/Athena/LinearColorCurves/VFX_CurveAtlas
   curve=/Game/Gadgets/Assets/VinderTech_GliderChute/Glider_Cube_Queen_Female/FX/Cube_Queen_StairMask_Cuve
   defaultValue=43
   parameterName=StairMask_Curve
  }
  n109 = Saturate(None=@n108.R)
  n110 = ActorPositionWS()
  n111 = Constant() {r=-20}
  n112 = Constant() {r=20}
  n113 = LinearInterpolate(Alpha=ParticleRandom()) {constA=5.7, constB=5.3}
  n114 = Time()
  n115 = TextureSample(UVs=Panner(Coordinate=TextureCoordinate(), Time=Add(A=@n113, B=@n114)) {speedY=0.5}) {samplerType=SAMPLERTYPE_Grayscale, texture=&LowResBlurredNoise}
  n116 = Transform(None=Add(
    A=VectorParameter() {group=CubeQueen_Stairs_Mask, parameterName=Stairs_Mask_Offset_Top, defaultValue=<rgba 0 -340 0 1>}.RGB
    B=LinearInterpolate(A=@n111, B=@n112, Alpha=@n115.R) {constA=-50, constB=0}
   ) {constB=0}) {transformSourceType=TRANSFORMSOURCE_Local}
  n117 = Constant() {r=500}
  n118 = Subtract(A=@n117, B=Constant() {r=50}) {constB=50}
  n119 = CurveAtlasRowParameter(CurveTime=SphereMask(
    A=WorldPosition().XYZ
    B=Add(A=@n110, B=@n116)
    Radius=LinearInterpolate(A=@n118, B=@n117, Alpha=@n115.R) {constA=250}
    Hardness=Constant() {r=0.5}
   )) {
   atlas=&VFX_CurveAtlas
   curve=&Cube_Queen_StairMask_Cuve
   defaultValue=43
   parameterName=StairMask_Curve
  }
  n120 = Saturate(None=@n119.R)
  n121 = StaticSwitchParameter(True=Multiply(A=Saturate(None=@n98), B=Subtract(A=@n109, B=@n120)), False=Constant() {r=1}) {group=Opacity, parameterName=Use_CubeQueen_Stairs_Mask}
  n122 = ParticleColor()
  n123 = TextureSampleParameter2D(UVs=Panner(Coordinate=TextureCoordinate() {uTiling=1.5}) {speedX=1.5}) {
   group=Trail
   parameterName=Base_Trail_texture
   samplerType=SAMPLERTYPE_LinearColor
   texture=/BRCosmetics/Athena/Effects/Particles/Skydiving/NS_Contrail_CubeNinja/T_CubeNinja_Trail
  }
  n124 = MaterialFunctionCall(Float2=TextureCoordinate()) {materialFunction=&BreakOutFloat2Components}
  n125 = Multiply(A=Multiply(A=@n122.A, B=Power(Base=@n123.R) {constExponent=1}), B=Power(Base=Sine(None=@n124.R) {period=2}))
  n126 = Multiply(
   A=StaticSwitchParameter(True=DepthFade(Opacity=@n125), False=Multiply(A=@n96, B=@n121)) {group=Trail, parameterName=Use_pickaxe_trail}
   B=MaterialFunctionCall(Value_Morning=@n30.A, Value_Day=@n31.A, Value_Evening=@n32.A, Value_Night=@n33.A) {materialFunction=&Scalar_Time_of_Day}.Result
  )
  n127 = ScalarParameter() {defaultValue=256, parameterName=NearCamera_Opacity}
  LinearInterpolate(
   A=@n126
   B=Multiply(A=@n126, B=MaterialFunctionCall(Distance=@n127) {materialFunction=&Near_Camera_Fade}.Result)
   Alpha=ScalarParameter() {parameterName=Use_Near_Camera_Fade}
  )
  n128 = ComponentMask(None=MaterialFunctionCall!(Gradient) {materialFunction=/BRCosmetics/Athena/Effects/Particles/Skydiving/NS_Contrail_DarkAndLight/MF_DarkAndLight_RainbowColors}.Result) {b=true, g=true, r=true}
  n129 = ScalarParameter() {parameterName=Fraction}
  n130 = VectorParameter() {parameterName=ColorA, defaultValue=<rgba 0 0.742 1 1>}
  n131 = VectorParameter() {parameterName=ColorB, defaultValue=<rgba 0.0997 0.045 1 1>}
  n132 = VectorParameter() {parameterName=ColorC, defaultValue=<rgba 0 0.0145 1 1>}
  n133 = Multiply(
   A=MaterialFunctionCall(Float2=TextureCoordinate()) {materialFunction=&BreakOutFloat2Components}.R
   B=ScalarParameter() {defaultValue=0.5, group=Trail, parameterName=Gradient_intensity}
  ) {constB=0.5}
  n134 = Multiply(
   A=@n122.RGB
   B=StaticSwitchParameter(
    True=Desaturation(None=@n128, Fraction=@n129)
    False=MaterialFunctionCall(A=@n130.RGB, B=@n131.RGB, C=@n132.RGB, Alpha=@n133) {materialFunction=&3ColorBlend}.Result
   ) {defaultValue=true, parameterName=UseRainbowColor}
  )
  n135 = TextureSampleParameter2D(UVs=@n6) {group=Base_color, parameterName=Base_color_lerp_mask, texture=&T_Inky_Smoke_Tile}
  n136 = TextureCoordinate()
  n137 = Constant2Vector() {g=0.5, r=0.5}
  n138 = Multiply(
   A=LinearInterpolate(
    A=VectorParameter() {group=Base_color, parameterName=Base_Color1, defaultValue=<rgba 0.8499 0.0466 1.0417 1>}.RGB
    B=VectorParameter() {group=Base_color, parameterName=Base_Color2, defaultValue=<rgba 0 1.2585 21.3542 1>}.RGB
    Alpha=@n135.R
   )
   B=Power(
    Base=SphereMask(A=@n136, B=@n137) {attenuationRadius=0.8, hardnessPercent=0}
    Exp=ScalarParameter() {defaultValue=3, group=Base_color, parameterName=Base_Texture_power}
   ) {constExponent=3}
  )
  n139 = CurveAtlasRowParameter(CurveTime=ComponentMask(None=TextureCoordinate()) {g=true}) {
   atlas=&VFX_CurveAtlas
   curve=/Game/Accessories/FORT_Backpacks/Backpack_F_MED_Tower_Sentinel/FX/Curve_Tower_Sentinel_Female
   defaultValue=15
   parameterName=Global_color
  }
  n140 = TextureCoordinate()
  n141 = Multiply(A=@n80, B=Constant() {r=10}) {constB=3}
  n142 = LinearInterpolate(A=Floor(None=@n141), B=@n141, Alpha=ScalarParameter() {defaultValue=0.3, group=Fire_center, parameterName=Fire_center_Hard_edge})
  n143 = Add(A=TextureCoordinate() {uTiling=0.5, vTiling=0.5}, B=NamedRerouteUsage())
  n144 = Constant() {r=1}
  n145 = DynamicParameter() {parameterIndex=2, paramNames [V_Scale,Param2,Param3,Param4]}
  n146 = Add(
   A=Multiply(A=TextureCoordinate(), B=AppendVector(A=@n144, B=@n145.V_Scale))
   B=DynamicParameter() {paramNames [Offset, Texture_mask_offset, WaveIntensity, Panner_offset]}.Panner_offset
  )
  n147 = ScalarParameter() {parameterName=Fix_the_edges}
  n148 = TextureSample(UVs=Panner(Coordinate=LinearInterpolate(A=@n143, B=@n146, Alpha=@n147), Time=NamedRerouteUsage()) {speedY=0.3}) {samplerType=SAMPLERTYPE_LinearColor, texture=/Game/Global/Textures/Default/Noise/T_Noise01_Inverse}
  n149 = VectorParameter() {group=Fire_center, parameterName=Fire_center_color02, defaultValue=<rgba 0 11.3982 50 1>}
  n150 = LinearInterpolate(
   B=Multiply(A=@n142, B=Multiply(A=@n148.R, B=@n149.RGB))
   Alpha=Power(Base=ComponentMask(None=@n140) {g=true}, Exp=ScalarParameter() {defaultValue=18, group=Fire_center, parameterName=Fire_center_power}) {constExponent=18}
  )
  n151 = Power(Base=ComponentMask(None=TextureCoordinate()) {g=true}) {constExponent=3}
  n152 = Add(A=TextureCoordinate() {uTiling=0.5, vTiling=0.5}, B=NamedRerouteUsage())
  n153 = DynamicParameter() {paramNames [Offset, Texture_mask_offset, WaveIntensity, Panner_offset]}
  n154 = TextureCoordinate()
  n155 = AppendVector(A=Constant() {r=1}, B=DynamicParameter() {parameterIndex=2, paramNames [V_Scale,Param2,Param3,Param4]}.V_Scale)
  n156 = LinearInterpolate(
   A=LinearInterpolate(A=@n152, B=Add(A=@n152, B=@n153.Panner_offset), Alpha=ScalarParameter() {group=Fire_center, parameterName=Use highlight tex random offset})
   B=Add(A=Multiply(A=@n154, B=@n155), B=@n153.Panner_offset)
   Alpha=ScalarParameter() {parameterName=Fix_the_edges}
  )
  n157 = NamedRerouteUsage()
  n158 = Multiply(
   A=Saturate(None=LinearInterpolate(B=@n142, Alpha=@n151))
   B=TextureSample(UVs=Panner(Coordinate=@n156, Time=@n157) {speedY=0.3}) {samplerType=SAMPLERTYPE_LinearColor, texture=/Game/Characters/Player/Male/Medium/Bodies/M_MED_SRDrift/Textures/T_ElectricNoise}.R
  )
  n159 = Desaturation(
   None=Multiply(A=VectorParameter() {group=Fire_center, parameterName=Fire_center_color, defaultValue=<rgba 7.3424 1.899 20 1>}.RGB, B=Power(Base=@n158))
   Fraction=ScalarParameter() {defaultValue=-0.5, group=Fire_center, parameterName=Fire_center_color_purple_Desaturation}
  )
  n160 = Add(
   A=LinearInterpolate(A=@n138, B=Multiply(A=@n138, B=@n139), Alpha=ScalarParameter() {defaultValue=1, parameterName=MultColorCurveForEmissive})
   B=Multiply(A=LinearInterpolate(A=@n150, B=@n159)) {constB=3}
  ) {constB=0}
  n161 = ParticleColor()
  n162 = Multiply(
   A=LinearInterpolate(A=@n160, B=Multiply(A=@n161.RGB, B=@n160), Alpha=ScalarParameter() {parameterName=Use_particle_color})
   B=ScalarParameter() {defaultValue=1, parameterName=Glider_appear_brightness}
  )
  MaterialFunctionCall(Emissive=Multiply(
    A=StaticSwitchParameter(True=@n134, False=@n162) {group=Trail, parameterName=Use_pickaxe_trail}
    B=MaterialFunctionCall(Color_Morning=@n30.RGB, Color_Day=@n31.RGB, Color_Evening=@n32.RGB, Color_Night=@n33.RGB) {materialFunction=&Color_Time_of_Day}.Result
   )) {materialFunction=&MF_FX_EmissiveAdjustments}
  n163 = Multiply(A=ScalarParameter() {defaultValue=1, group=Panner Speed, parameterName=Speed 2}, B=Time())
  NamedRerouteDeclaration(None=@n163) {name=Speed 2 Param, nodeColor=<rgba 0 1 0.4141 1>}
  n164 = Constant()
  n165 = DynamicParameter() {paramNames [Offset, Texture_mask_offset, WaveIntensity, Panner_offset]}
  NamedRerouteDeclaration(None=LinearInterpolate(B=AppendVector(A=@n164, B=@n165.Offset), Alpha=ScalarParameter() {parameterName=UseCustomOffset})) {name=CustomOffset, nodeColor=<rgba 1 0.7734 0 1>}
  Reroute(None=@n163)
  n166 = DynamicParameter() {desc=Parameters set by Niagara, parameterIndex=1, paramNames [X,Y,Z,WPOIntensity]}
  n167 = MaterialFunctionCall(X=@n166.X, Y=@n166.X, Z=@n166.X) {materialFunction=&MakeFloat3}
  n168 = ParticlePositionWS()
  n169 = ComponentMask(None=OneMinus(None=TextureCoordinate())) {g=true}
  n170 = ScalarParameter() {defaultValue=3, group=WPO, parameterName=WPOBendPower}
  n171 = LinearInterpolate(
   B=Clamp(
    None=Subtract(A=@n167.Result, B=@n168)
    Min=ScalarParameter() {defaultValue=-100, group=WPO, parameterName=Clamp_min}
    Max=ScalarParameter() {defaultValue=100, group=WPO, parameterName=Clamp_max}
   ) {maxDefault=1000, minDefault=-1000}
   Alpha=Multiply(A=Power(Base=@n169, Exp=@n170), B=@n166.WPOIntensity)
  )
  n172 = Constant()
  n173 = OneMinus(None=MaterialFunctionCall(Distance=Constant() {r=3000}) {materialFunction=&Near_Camera_Fade}.Result)
  n174 = Add(A=TextureCoordinate(), B=NamedRerouteUsage())
  n175 = LinearInterpolate(
   B=AppendVector(A=Constant(), B=DynamicParameter() {paramNames [Offset, Texture_mask_offset, WaveIntensity, Panner_offset]}.Panner_offset)
   Alpha=ScalarParameter() {parameterName=Use WPO Offset}
  )
  n176 = VectorParameter() {group=WPO, parameterName=WPO_UVScale, defaultValue=<rgba 0 1 0 0>}
  n177 = Add(A=Multiply(A=Add(A=@n174, B=@n175), B=ComponentMask(None=@n176.RGB) {g=true, r=true}), B=AppendVector(A=Constant(), B=VertexColor().R)) {constB=0}
  n178 = ComponentMask(None=VectorParameter() {group=WPO, parameterName=WPO_Speed, defaultValue=<rgba 0 0.5 0 0>}.RGB) {g=true, r=true}
  n179 = LinearInterpolate(Alpha=TextureSample(UVs=Panner(Coordinate=@n177, Speed=@n178) {speedY=0.5}) {samplerType=SAMPLERTYPE_Normal, texture=&T_Weap_Distort_N_02}.R) {constB=1.5}
  n180 = OneMinus(None=ComponentMask(None=TextureCoordinate()) {g=true})
  n181 = Saturate(None=Multiply(A=Power(Base=@n180) {constExponent=1.5}))
  n182 = Transform(None=Multiply(A=Multiply(A=@n179, B=@n181), B=VectorParameter() {group=WPO, parameterName=WPO_direction, defaultValue=<rgba 1 0 0 1>}.RGB))
  n183 = Multiply(
   A=@n182
   B=LinearInterpolate(
    A=ScalarParameter() {defaultValue=20, group=WPO, parameterName=Glider_Lod_low_WPO_intensity}
    B=ScalarParameter() {defaultValue=80, group=WPO, parameterName=Glider_WPO_intensity}
    Alpha=Saturate(None=@n173)
   )
  ) {constB=80}
  n184 = ScalarParameter() {defaultValue=15, group=WPO, parameterName=Glider_Moble_Lod_low_WPO_intensity}
  n185 = ScalarParameter() {defaultValue=80, group=WPO, parameterName=Glider_WPO_intensity}
  n186 = Constant() {r=1000}
  n187 = Saturate(None=OneMinus(None=MaterialFunctionCall(Distance=@n186) {materialFunction=&Near_Camera_Fade}.Result))
  n188 = ShadingPathSwitch(Default=@n183, Deferred=@n183, Forward=@n183, Mobile=Multiply(A=@n182, B=LinearInterpolate(A=@n184, B=@n185, Alpha=@n187)))
  n189 = OneMinus(None=MaterialFunctionCall(Distance=Constant() {r=1700}) {materialFunction=&Near_Camera_Fade}.Result)
  n190 = Multiply(
   A=@n182
   B=LinearInterpolate(
    A=ScalarParameter() {defaultValue=20, group=WPO, parameterName=Lod_low_WPO_intensity}
    B=ScalarParameter() {defaultValue=80, group=WPO, parameterName=WPO_intensity}
    Alpha=Saturate(None=@n189)
   )
  ) {constB=80}
  n191 = ScalarParameter() {defaultValue=15, group=WPO, parameterName=Moble_Lod_low_WPO_intensity}
  n192 = ScalarParameter() {defaultValue=80, group=WPO, parameterName=WPO_intensity}
  n193 = Constant() {r=500}
  n194 = Saturate(None=OneMinus(None=MaterialFunctionCall(Distance=@n193) {materialFunction=&Near_Camera_Fade}.Result))
  n195 = ShadingPathSwitch(Default=@n190, Deferred=@n190, Forward=@n190, Mobile=Multiply(A=@n182, B=LinearInterpolate(A=@n191, B=@n192, Alpha=@n194)))
  StaticSwitchParameter(
   True=Constant()
   False=Add(
    A=StaticSwitchParameter(True=@n171, False=@n172) {group=WPO, parameterName=Use_Interpolate WPO}
    B=StaticSwitchParameter(True=@n188, False=@n195) {group=WPO, parameterName=Use_Glider_Lod_WPO_Fix}
   )
  ) {group=Trail, parameterName=Use_pickaxe_trail}
 ]
}
